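# Physics asset for the lower mesh of the Splinter backpack cosmetic, defining capsule bodies and joints for its dynamic fur bones.
# Left and right master fur bodies are constrained to a small pack body, with two fur bodies hanging off the left master and four off the right.
# Angular limits are tighter at the master-to-pack joints than at the outer fur joints, and the right master uses two capsules along its length.
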PhysicsAsset /BRCosmetics/Accessories/FORT_Backpacks/Backpack_M_MED_Splinter/Meshes/M_MED_Splinter_Pack_Lower_Physics
bodies {
 dyn_fur_l1 [{shapeType=Capsule, length=3, radius=3, center=<1 0 0>, rotation=<rot 90 0 0>}]
 dyn_fur_l2 [{shapeType=Capsule, length=3, radius=2.25, center=<2 0 0>, rotation=<rot 90 0 0>}]
 dyn_fur_r1 [{shapeType=Capsule, length=2.25, radius=1.5, center=<2 0 0>, rotation=<rot 90 0 0>}]
 dyn_fur_r2 [{shapeType=Capsule, length=2.5, radius=2.25, center=<0.5 0 0>, rotation=<rot 90 0 0>}]
 dyn_fur_r3 [{shapeType=Capsule, length=2.75, radius=3, center=<0.75 0 0>, rotation=<rot 90 0 0>}]
 dyn_fur_r6 [{shapeType=Capsule, length=2.75, radius=3, center=<0.75 0 0>, rotation=<rot 90 0 0>}]
 dyn_master_fur_l [{shapeType=Capsule, length=3, radius=3.5, center=<1 0 0>, rotation=<rot 90 0 0>}]
 dyn_master_fur_r [
  {shapeType=Capsule, length=3.5, radius=4.25, center=<2.25 0 0>, rotation=<rot 90 0 0>}
  {shapeType=Capsule, length=3.5, radius=4.25, center=<11.2323 0 0>, rotation=<rot 90 0 0>}
 ]
 pack [{shapeType=Capsule, length=1, radius=1}]
}
constraints [
 {bone1Name=dyn_fur_r1, bone2Name=dyn_master_fur_r, swing1LimitDegrees=45, swing2LimitDegrees=45, twistLimitDegrees=45}
 {bone1Name=dyn_master_fur_r, bone2Name=pack, swing1LimitDegrees=40, swing2LimitDegrees=30, twistLimitDegrees=30}
 {bone1Name=dyn_fur_r2, bone2Name=dyn_master_fur_r, swing1LimitDegrees=45, swing2LimitDegrees=45, twistLimitDegrees=45}
 {bone1Name=dyn_fur_r3, bone2Name=dyn_master_fur_r, swing1LimitDegrees=45, swing2LimitDegrees=45, twistLimitDegrees=45}
 {bone1Name=dyn_fur_r6, bone2Name=dyn_master_fur_r, swing1LimitDegrees=45, swing2LimitDegrees=45, twistLimitDegrees=45}
 {bone1Name=dyn_master_fur_l, bone2Name=pack, swing1LimitDegrees=30, swing2LimitDegrees=30, twistLimitDegrees=30}
 {bone1Name=dyn_fur_l1, bone2Name=dyn_master_fur_l, swing1LimitDegrees=45, swing2LimitDegrees=45, twistLimitDegrees=45}
 {bone1Name=dyn_fur_l2, bone2Name=dyn_master_fur_l, swing1LimitDegrees=45, swing2LimitDegrees=45, twistLimitDegrees=45}
]
