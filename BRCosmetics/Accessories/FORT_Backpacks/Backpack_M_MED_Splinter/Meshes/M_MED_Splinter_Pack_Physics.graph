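# Physics asset for the M_MED_Splinter backpack cosmetic, defining collision bodies and joints for its dynamic bones.
# Capsule bodies sit on eight left and right dyn_master_upper bones, plus a small capsule on the pack bone.
# Each dynamic bone is constrained directly to pack in a star layout rather than a chain, with modest swing limits and tight twist limits that restrict its movement relative to the pack.
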PhysicsAsset /BRCosmetics/Accessories/FORT_Backpacks/Backpack_M_MED_Splinter/Meshes/M_MED_Splinter_Pack_Physics
bodies {
 dyn_master_upper_l [{shapeType=Capsule, length=3.5, radius=2.255, center=<1.5 0 0>, rotation=<rot 90 0 0>}]
 dyn_master_upper_l1 [{shapeType=Capsule, length=1.5, radius=1.75, center=<2 0 0>, rotation=<rot 90 0 0>}]
 dyn_master_upper_l2 [{shapeType=Capsule, length=3.25, radius=2.25, center=<2.5 0 0>, rotation=<rot 90 0 0>}]
 dyn_master_upper_r [{shapeType=Capsule, length=2, radius=1.5, center=<2 0 0>, rotation=<rot 90 0 0>}]
 dyn_master_upper_r1 [{shapeType=Capsule, length=1.75, radius=2, center=<1.1184 0 0>, rotation=<rot 90 0 0>}]
 dyn_master_upper_r3 [{shapeType=Capsule, length=2.75, radius=3.25, center=<1.5 0 0>, rotation=<rot 90 0 0>}]
 dyn_master_upper_r4 [{shapeType=Capsule, length=4, radius=3.25, center=<1 0 0>, rotation=<rot 90 0 0>}]
 dyn_master_upper_r5 [{shapeType=Capsule, length=2.25, radius=2.5, center=<0.65 0 0>, rotation=<rot 90 0 0>}]
 pack [{shapeType=Capsule, length=1, radius=1}]
}
constraints [
 {bone1Name=dyn_master_upper_l, bone2Name=pack, swing1LimitDegrees=15, swing2LimitDegrees=15, twistLimitDegrees=3}
 {bone1Name=dyn_master_upper_l1, bone2Name=pack, swing1LimitDegrees=25, swing2LimitDegrees=25, twistLimitDegrees=5}
 {bone1Name=dyn_master_upper_l2, bone2Name=pack, swing1LimitDegrees=20, swing2LimitDegrees=20, twistLimitDegrees=5}
 {bone1Name=dyn_master_upper_r4, bone2Name=pack, swing1LimitDegrees=20, swing2LimitDegrees=15, twistLimitDegrees=5}
 {bone1Name=dyn_master_upper_r3, bone2Name=pack, swing1LimitDegrees=30, swing2LimitDegrees=35, twistLimitDegrees=5}
 {bone1Name=dyn_master_upper_r, bone2Name=pack, swing1LimitDegrees=20, swing2LimitDegrees=20, twistLimitDegrees=5}
 {bone1Name=dyn_master_upper_r1, bone2Name=pack, swing1LimitDegrees=20, swing2LimitDegrees=20, twistLimitDegrees=5}
 {bone1Name=dyn_master_upper_r5, bone2Name=pack, swing1LimitDegrees=20, swing2LimitDegrees=20, twistLimitDegrees=5}
]
